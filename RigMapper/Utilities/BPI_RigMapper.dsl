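; Blueprint interface for the RigMapper utilities, which lets implementing objects opt into rig mapping.
; It declares EnableRigMapper, which returns a flag, and OverrideDefinitions, which returns a flag plus a second output, likely the override definitions.
; Both functions carry only default false or empty returns, so the actual behavior comes from implementers.
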(fn EnableRigMapper
  (return false))

(fn OverrideDefinitions
  (return false 0))
