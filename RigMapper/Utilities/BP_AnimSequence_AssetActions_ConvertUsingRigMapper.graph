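# Editor utility Blueprint that batch-converts selected animation sequences using Rig Mapper definitions.
# Convert Selected Using RigMapper takes a definition array, a target skeletal mesh and an output suffix, defaulting to "_Converted".
# It loops over selected AnimSequenceBase assets, casts each to AnimSequence and passes it to ConvertAnimSequenceNew.
# Each new asset's folder and name derive from the source package path, with the suffix appended to the name.
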EditorUtilityBlueprint /RigMapper/Utilities/BP_AnimSequence_AssetActions_ConvertUsingRigMapper
graphs [
 Convert Selected Using RigMapper {nodes [
   n1 = Convert Selected Using RigMapper!() -> (
    then
    RigMapper Definitions:Array of Rig Mapper Definition Object References
    Output Suffix:String=_Converted
    Target Mesh:Skeletal Mesh Object Reference
   ) {is=title K2Node_FunctionEntry}
   > n2 = GetSelectedAssetsOfClass(AssetClass=/Script/Engine.AnimSequenceBase)
   > n3 = For Each Loop(Array:Array of Object References=@n2) -> (Array Element:Object Reference)
   > n4 = Cast To AnimSequence!(execute, Object=@n3) -> (then, CastFailed, AsAnim Sequence:Anim Sequence Object Reference)
   n5 = GetDisplayName(Object=GetPackageForObject!(Object:Object Reference=@n3) -> (ReturnValue:Package Object Reference))
   n6 = GetBaseFilename(InPath=@n5, bRemovePath=true)
   > ConvertAnimSequenceNew!(
    execute
    Source:Anim Sequence Object Reference=@n4
    TargetMesh:Skeletal Mesh Object Reference=@n1.Target Mesh
    Definitions:Array of Rig Mapper Definition Object References=@n1
    NewAssetPath:"Directory Path Structure (by ref)"=Make Directory Path!(Path:String=GetPath!(InPath:String=@n5) -> (ReturnValue:String)) -> (DirectoryPath:Directory Path Structure) {is=title K2Node_MakeStruct}
    NewAssetName:Name=String To Name(InString=Append(A=@n6, B=@n1.Output Suffix))
   ) -> (then, ReturnValue:Anim Sequence Object Reference)
  ]}
 EventGraph {links [], nodes []}
]
variables []
