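# Editor utility that adds ExportToJson, LoadFromJson and ValidateAssets asset actions for Rig Mapper Definition assets.
# Each function loops over the selected definitions, casts each, calls its native export, load or validate method, and prints success or failure.
# Export and load take a JSON file path as input and return inside the loop after the first print, so only the first selected asset is processed.
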EditorUtilityBlueprint /RigMapper/Utilities/BP_RigMapperDefinition_AssetActions
graphs [
 EventGraph {links [], nodes []}
 ExportToJson {nodes [
   n1 = ExportToJson!() -> (then, JsonFile:File Path Structure) {is=title K2Node_FunctionEntry}
   > n2 = GetSelectedAssetsOfClass(AssetClass=/Script/RigMapper.RigMapperDefinition)
   > n3 = For Each Loop(Array:Array of Object References=@n2) -> (Array Element:Object Reference)
   > n4 = Cast To RigMapperDefinition!(execute, Object=@n3) -> (then, CastFailed, AsRig Mapper Definition:Rig Mapper Definition Object Reference) {is=title K2Node_DynamicCast}
   n5 = ExportAsJsonFile!(self:Rig Mapper Definition Object Reference=@n4, JsonFilePath:"File Path Structure (by ref)"=@n1) -> (ReturnValue:bool)
   > PrintString(InString=Append!(
     A="Exporting: "
     B=GetDisplayName(Object=@n3)
     C:String=" from: "
     D:String=Break File Path!(FilePath:"File Path Structure (by ref)"=@n1) -> (FilePath:String) {is=title K2Node_BreakStruct}
     E:String=": "
     F:String=SelectString(A=Success, B=Failure, bPickA=@n5)
    ) -> (ReturnValue))
   > Return Node()
  ]}
 LoadFromJson {nodes [
   n1 = LoadFromJson!() -> (then, JsonFile:File Path Structure) {is=title K2Node_FunctionEntry}
   > n2 = GetSelectedAssetsOfClass(AssetClass=&RigMapperDefinition)
   > n3 = For Each Loop(Array=@n2)
   > n4 = Cast To RigMapperDefinition(Object=@n3)
   > n5 = LoadFromJsonFile!(execute, self:Rig Mapper Definition Object Reference=@n4, JsonFilePath:"File Path Structure (by ref)"=@n1) -> (then, ReturnValue:bool)
   > PrintString(InString=Append(
     A="Loading: "
     B=GetDisplayName(Object=@n3)
     C=" from: "
     D=Break File Path(FilePath=@n1)
     E=": "
     F=SelectString(A=Success, B=Failure, bPickA=@n5)
    ))
   > Return Node()
  ]}
 ValidateAssets {nodes [
   ValidateAssets() {is=title K2Node_FunctionEntry}
   > n1 = GetSelectedAssetsOfClass(AssetClass=&RigMapperDefinition)
   > n2 = For Each Loop(Array=@n1)
   > n3 = Cast To RigMapperDefinition(Object=@n2)
   > n4 = Validate!(execute, self:Rig Mapper Definition Object Reference=@n3) -> (then, ReturnValue:bool)
   > PrintString(InString=Append!(A="Validating ", B=GetDisplayName(Object=@n2), C=": ", D=SelectString(A=Success, B=Failure, bPickA=@n4)) -> (ReturnValue))
  ]}
]
variables []
